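; Burst gameplay cue notify for the Thermal Lance pulse explosion landing, overriding OnBurst to keep the decal off moving targets such as trains.
; After the parent spawns the effects, it reads the hit actor from the effect context's hit result.
; If that actor's root component is Movable, it sets the IsMoving parameter on the first spawned FX component, leaving decal suppression to the particle system.
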(fn OnBurst
  (params (Target :type Actor%20Object%20Reference) (Parameters :type Gameplay%20Cue%20Parameters%20Structure%20%28by%20ref%29) (SpawnResults :type Gameplay%20Cue%20Notify%20Spawn%20Result%20Structure%20%28by%20ref%29))
  (bind _fxsystemcomponents (Utilities|Struct|BreakGameplayCueNotifySpawnResult SpawnResults))
  (bind (_normalizedmagnitude _rawmagnitude _effectcontext) (Ability|GameplayCue|BreakGameplayCueParameters Parameters))
  (bind _returnvalue (Ability|EffectContext|GetHitResult _effectcontext))
  (bind (_bblockinghit _binitialoverlap _time _distance _location _impactpoint _normal _impactnormal _physmat _hitactor) (Collision|BreakHitResult _returnvalue))
  (bind _output (Utilities|Array|Get(acopy) _fxsystemcomponents))
  ;[0.1116 0.1119 0.15 0.5] OnBurst: Prevent the decal from spawning if burst occurs on a movable target, like a train.
  (|CallParentFunction Target Parameters SpawnResults)
  (Utilities|FlowControl|Sequence
    (:then_0
      (if (Utilities|Array|IsValidIndex _fxsystemcomponents)
        (Utilities|IsValid _output
          (:Is%20Valid
            (Utilities|IsValid _hitactor
              (:Is%20Valid
                (bind _rootcomponent (Variables|/Script/Engine.Actor|GetRootComponent _hitactor)
                  (:then
                    (if (Utilities|Enum|Equal(Enum) (Variables|/Script/Engine.SceneComponent|GetMobility _rootcomponent) "Movable")
                      (Effects|Components|ParticleSystem|SetBoolParameter _output "IsMoving" true)))
                  (:else))))))))
    (:then_1
      (return))))
