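; Parent burst gameplay cue notify for the Thermal Lance pulse explosion, handling its effects and audio.
; OnBurst runs the parent logic, then sets ExplosionRadius on spawned Niagara components from a cached scalable float.
; If the cue location is non-zero, it spawns the explosion sound there and registers a sound indicator.
; The sound receives player, water and vault boolean parameters from the cue's variables.
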(fn OnBurst
  (params (Target :type Actor%20Object%20Reference) (Parameters :type Gameplay%20Cue%20Parameters%20Structure%20%28by%20ref%29) (SpawnResults :type Gameplay%20Cue%20Notify%20Spawn%20Result%20Structure%20%28by%20ref%29))
  (bind (_normalizedmagnitude _rawmagnitude _effectcontext _matchedtagname _originaltag _aggregatedsourcetags _aggregatedtargettags _location) (Ability|GameplayCue|BreakGameplayCueParameters Parameters))
  (|CallParentFunction Target Parameters SpawnResults)
  (Utilities|FlowControl|Sequence
    (:then_0
      (CallFunction|UpdateSpawnedFXSystemComponents (Utilities|Struct|BreakGameplayCueNotifySpawnResult SpawnResults))
      (if (Utilities|Operators|NotEqual(!=) :A _location)
        (bind _returnvalue (Audio|SpawnSoundatLocation Target (Variables|Default|GetExplosionSound) _location))
        (Utilities|IsValid _returnvalue
          (:Is%20Valid
            (Default|SpawnSoundIndicator (Variables|GetTarget) (Utilities|Array|MakeArray _returnvalue) _location)
            (Audio|Parameter|SetBooleanParameter _returnvalue "isPlayer" (Variables|Default|GetisPlayer) :!class /Script/AudioExtensions.AudioParameterControllerInterface)
            (Audio|Parameter|SetBooleanParameter _returnvalue "isWater" (Variables|Default|GetisWater) :!class /Script/AudioExtensions.AudioParameterControllerInterface)
            (Audio|Parameter|SetBooleanParameter _returnvalue "isVault" (Variables|Default|GetisVault) :!class /Script/AudioExtensions.AudioParameterControllerInterface))
          (:Is%20Not%20Valid
            (return)))))
    (:then_1
      (return))))

(fn UpdateSpawnedFXSystemComponents
  (params (FXSystemComponents :type Array%20of%20FXSystem%20Component%20Object%20References))
  (for _array_element FXSystemComponents
    (bind _asniagara_particle_system_component (Utilities|Casting|CastToNiagaraComponent _array_element))
    (Niagara|SetNiagaraVariable(Float) _asniagara_particle_system_component "ExplosionRadius" (ScalableFloat|GetCachedScalableFloatValue (Variables|Default|GetPulseExplosionRadius)))))
