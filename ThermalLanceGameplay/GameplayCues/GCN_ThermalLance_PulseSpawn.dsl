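; Burst gameplay cue that plays the cosmetics when the Thermal Lance fires a pulse.
; On burst, it casts the target to a Fort player pawn and its current weapon to the Thermal Lance weapon.
; It then has the weapon spawn the pulse cosmetic on the beam Niagara effect and plays a local camera lens effect and camera shake on that pawn.
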(event EventOnBurst
  (params (Target :type Actor%20Object%20Reference) (Parameters :type Gameplay%20Cue%20Parameters%20Structure%20%28by%20ref%29) (SpawnResults :type Gameplay%20Cue%20Notify%20Spawn%20Result%20Structure%20%28by%20ref%29))
  ;[0.004897 0.1049 0.15 0.5] OnBurst: When the pulse is fired, activate the pulse on the beam niagara effect.
  (|CallParentFunction Target Parameters SpawnResults)
  (bind _asfort_player_pawn (Utilities|Casting|CastToFortPlayerPawn Target))
  (bind _asb_weapon_thermal_lance (Utilities|Casting|CastToB_Weapon_ThermalLance (Weapon|GetCurrentWeapon _asfort_player_pawn)))
  (Class|BWeaponThermalLance|SpawnPulseCosmetic _asb_weapon_thermal_lance)
  (CallFunction|PlayLocalCameraLensEffect _asfort_player_pawn "/ThermalLanceGameplay/Assets/VFX/BP_CameraLens_ThermalLance_Pulse.BP_CameraLens_ThermalLance_Pulse_C")
  (CallFunction|PlayLocalCameraShake _asfort_player_pawn "/ThermalLanceGameplay/Assets/VFX/BP_CameraShake_ThermalLance_Throw.BP_CameraShake_ThermalLance_Throw_C"))
